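; Editor-only foliage cluster actor that describes its instanced meshes as simple destructible props of the Foliage resource type.
; It returns up to five instanced static mesh components, filtered by per-cluster enable flags, with one fixed Wood-typed entry for all.
; Its overrides decline keeping the source actor and using dynamic destruction materials or Niagara systems.
; Construction defers to the parent, and BeginPlay destroys the actor so it never exists on the PIE server.
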(event EventBeginPlay
  (|CallParentFunction)
  (Actor|DestroyActor) ; This actor is intended to only be used in the editor.\nWe destroy in BeginPlay to make sure we don't exist on the PIE server.
)

(event Collision|EventActorBeginOverlap
  (params (OtherActor :type Actor%20Object%20Reference)))

(event EventTick
  (params (DeltaSeconds :type Float%20%28single-precision%29)))

(fn GetDynamicDestructionResourceType
  (return "Foliage"))

(fn GetSimpleDestructiblePropComponents
  (Utilities|FlowControl|Sequence
    (:then_0
      (if (Variables|Cluster|GetEnableCluster0)
        (Utilities|Array|Add (LocalVariables|GetEnabledISMComponents) (Variables|Default|GetInstancedStaticMesh0))))
    (:then_1
      (if (Variables|Cluster|GetEnableCluster1)
        (Utilities|Array|Add (LocalVariables|GetEnabledISMComponents) (Variables|Default|GetInstancedStaticMesh1))))
    (:then_2
      (if (Variables|Cluster|GetEnableCluster2)
        (Utilities|Array|Add (LocalVariables|GetEnabledISMComponents) (Variables|Default|GetInstancedStaticMesh2))))
    (:then_3
      (if (Variables|Cluster|GetEnableCluster3)
        (Utilities|Array|Add (LocalVariables|GetEnabledISMComponents) (Variables|Default|GetInstancedStaticMesh3))))
    (:then_4
      (if (Variables|Cluster|GetEnableCluster4)
        (Utilities|Array|Add (LocalVariables|GetEnabledISMComponents) (Variables|Default|GetInstancedStaticMesh4))))
    (:then_5
      (return (LocalVariables|GetEnabledISMComponents)))))

(fn GetSimpleDestructiblePropEntry
  (params (SourceIndex :type Integer) (SourceComponent :type Static%20Mesh%20Component%20Object%20Reference))
  (return 0 0 "(Name=\"\")" 0 "(CurveTable=None,RowName=\"\")" "Wood" "Wood" "ATTACH_None" "SLOT_None" false false false))

(fn ShouldKeepSourceActor
  (return false))

(fn ShouldUseDynamicDestructionMaterial
  (return false))

(fn ShouldUseDynamicDestructionNiagaraSystem
  (return false))

(fn ConstructionScript
  (|CallParentFunction))
